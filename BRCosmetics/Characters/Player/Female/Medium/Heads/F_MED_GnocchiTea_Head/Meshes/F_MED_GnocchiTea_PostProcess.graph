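# Post-process Anim Blueprint for the GnocchiTea female medium head on the Fortnite base head skeleton, adjusting the under-chin facial bones.
# The AnimGraph chains three Modify Bone nodes, adding a fixed translation to the center, left and right bones at full alpha.
# The EventGraph eases Celeb Up Alpha by whether the female celebration montage plays, but the AnimGraph never reads it.
# A Copy Pose From Mesh node sits unconnected.
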AnimBlueprint /BRCosmetics/Characters/Player/Female/Medium/Heads/F_MED_GnocchiTea_Head/Meshes/F_MED_GnocchiTea_PostProcess
target_skeleton=/Game/Characters/Player/Common/Fortnite_Base_Head/Fortnite_Base_3L_Head_Skeleton
graphs [
 AnimGraph {nodes [
   Copy Pose From Mesh()
   n1 = "Transform (Modify) Bone - Bone: FACIAL_C_UnderChin"(ComponentPose=Local To Component(LocalPose=Input Pose()), Alpha=1, Translation="0,-5,5", Rotation="0,0,0", Scale="1,1,1") {translationMode=BMM_Additive, alphaBoolBlend {blendOption=Linear}, boneToModify {boneName=FACIAL_C_UnderChin}, translation {y=-5, z=5}}
   n2 = "Transform (Modify) Bone - Bone: FACIAL_L_UnderChin"(ComponentPose=@n1, Alpha=1, Translation="0,-5,5", Rotation="0,0,0", Scale="1,1,1") {translationMode=BMM_Additive, alphaBoolBlend {blendOption=Linear}, boneToModify {boneName=FACIAL_L_UnderChin}, translation {y=-5, z=5}}
   n3 = "Transform (Modify) Bone - Bone: FACIAL_R_UnderChin"(ComponentPose=@n2, Alpha=1, Translation="0,-5,5", Rotation="0,0,0", Scale="1,1,1") {translationMode=BMM_Additive, alphaBoolBlend {blendOption=Linear}, boneToModify {boneName=FACIAL_R_UnderChin}, translation {y=-5, z=5}}
   Output Pose(Result=Component To Local(ComponentPose=@n3))
  ]}
 EventGraph {nodes [
   n1 = Event BlueprintUpdateAnimation()
   n2 = GetOwnerSkeletalMeshComponent()
   > n3 = Is Valid(InputObject=@n2)
   n4 = GetAnimInstance(self=@n2)
   n5 = FInterpTo(
    Current=Get Celeb Up Alpha() -> (Celeb Up Alpha:double)
    Target="To Float (Boolean)"(InBool=Montage_IsPlaying(self=@n4, Montage=/BRCosmetics/Animation/Game/MainPlayer/Emotes/Emote_Celebration_Female_M))
    DeltaTime=float * float(A=@n1.DeltaTimeX, B=0.3) {is=title K2Node_PromotableOperator}
    InterpSpeed=30
   )
   > Set Celeb Up Alpha(Celeb Up Alpha:double=@n5) -> (Output_Get:double)
   Set Celeb Up Alpha(execute=@n3.Is Not Valid, Celeb Up Alpha=@n5)
   TryGetPawnOwner()
  ]}
]
variables [Celeb Up Alpha {type=number}]
